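# Master UI-domain material, named for Sapling enemy event icons, with translucent blending.
# Its only node is a texture sample parameter named Texture, which lets material instances swap the icon image and defaults to the CombatMode gamepad-option icon.
# The graph shows no math and no link from that sample to a material output.
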Material /RealitySeedGameplay/UI/Icons/M_UI_SaplingEnemyEventIcon_Master
graph {
 properties {blendMode=BLEND_Translucent, materialDomain=MD_UI, pixelDepthOffsetMode=PDOM_Legacy, usedShadingModels=Default Lit}
 nodes [TextureSampleParameter2D() {parameterName=Texture, texture=/Game/UI/Foundation/Textures/Icons/GamepadOptions/CombatMode}]
}
