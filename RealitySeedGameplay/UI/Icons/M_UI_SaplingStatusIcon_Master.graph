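# Translucent UI master material for sapling status icons, drawing a tintable icon with a soft-opacity drop shadow from the NeedsWeeding texture.
# UVs scale about the center, and a static component mask picks the icon channel, red by default, which is multiplied by Alpha.
# A second, offset sample scaled by ShadowAmount adds shadow coverage to opacity.
# The icon mask blends ShadowColor into Color.
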Material /RealitySeedGameplay/UI/Icons/M_UI_SaplingStatusIcon_Master
graph {
 properties {blendMode=BLEND_Translucent, materialDomain=MD_UI, pixelDepthOffsetMode=PDOM_Legacy, usedShadingModels=Default Lit}
 nodes [
  n1 = MaterialFunctionCall(Texture Scale=ScalarParameter() {defaultValue=0.85, parameterName=TextureScale}) {materialFunction=&ScaleUVsByCenter}
  n2 = TextureSample(UVs=@n1.UVs) {samplerSource=SSM_Clamp_WorldGroupSettings, samplerType=SAMPLERTYPE_LinearColor, texture=&/T_UI_NeedsWeeding_64x}
  n3 = Saturate(None=Multiply(A=StaticComponentMaskParameter(None=@n2.RGB) {defaultR=true, parameterName=Icon}, B=ScalarParameter() {defaultValue=1, parameterName=Alpha}))
  LinearInterpolate(
   A=VectorParameter() {parameterName=ShadowColor, defaultValue=<rgba 0 0 0 1>}.RGB
   B=VectorParameter() {parameterName=Color, defaultValue=<rgba 1 1 1 1>}.RGB
   Alpha=@n3
  )
  n4 = AppendVector(
   A=ScalarParameter() {defaultValue=-0.05, parameterName=ShadowOffset_X, sliderMin=-0.05}
   B=ScalarParameter() {defaultValue=-0.05, parameterName=ShadowOffset_Y}
  )
  n5 = StaticComponentMaskParameter(None=TextureSample(UVs=Add(A=@n1.UVs, B=@n4)) {samplerType=SAMPLERTYPE_LinearColor, texture=&T_UI_NeedsWeeding_64x}.RGB) {defaultR=true, parameterName=Icon}
  n6 = ScalarParameter() {defaultValue=0.125, parameterName=ShadowAmount}
  Saturate(None=Add(A=@n3, B=Multiply(A=@n5, B=@n6)))
 ]
}
